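# Sound cue for the meteor event's ambient sound, routed to the high-priority events sound class.
# It has no sound nodes or root, so as saved it plays nothing.
# Its attenuation override uses a custom long-range curve that is silent up close, peaks about a quarter of the way out and fades to zero.
# Interpolated listener focus with a narrow cone drops volume to zero when the listener faces away.
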SoundCue /Figment_S10/Athena/Sounds/Events/Meteor/Meteor_Ambient_01_Cue
root=null
nodes []
properties {
 bOverrideAttenuation=true
 maxDistance=101024
 soundClassObject=/Figment_S10/Athena/Sounds/Events/SoundClasses/Fort_Events_Priority_High
 volumeMultiplier=1
 attenuationOverrides {
  bEnableFocusInterpolation=true
  bEnableListenerFocus=true
  distanceAlgorithm=Custom
  falloffDistance=100000
  focusAttackInterpSpeed=0.1
  focusAzimuth=2
  lPFFrequencyAtMax=500
  lPFRadiusMax=2000
  lPFRadiusMin=1000
  nonFocusAzimuth=5
  nonFocusVolumeAttenuation=0
  stereoSpread=2048
  attenuationShapeExtents {x=1024}
  customAttenuationCurve {editorCurveData {keys [<at 0.1 0>,<at 0.25 1>,<at 1 0>]}}
 }
}
