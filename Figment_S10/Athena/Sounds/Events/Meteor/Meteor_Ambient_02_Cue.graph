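# Ambient sound cue for the meteor event, mixing a continuous meteor bed with intermittent distant rift-energy sounds.
# The looping bed runs through a repeating minute-long volume envelope that gently swells and falls.
# The second branch endlessly picks one of four borrowed RiftBeacon sounds after random delays.
# Each branch has its own long-range attenuation override, with the rift branch low-pass filtered almost from the source to sound muffled.
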SoundCue /Figment_S10/Athena/Sounds/Events/Meteor/Meteor_Ambient_02_Cue
root=Mixer_1
nodes [
 {
  id=Attenuation_0
  children [Looping_0]
  properties {
   bOverrideAttenuation=true
   attenuationOverrides {
    bAttenuateWithLPF=true
    falloffDistance=20000
    lPFFrequencyAtMax=500
    lPFFrequencyAtMin=500
    lPFRadiusMax=5
    lPFRadiusMin=0
    stereoSpread=4096
   }
  }
 }
 {
  id=Attenuation_1
  children [Enveloper_0]
  properties {bOverrideAttenuation=true, attenuationOverrides {falloffDistance=20000, stereoSpread=4096, attenuationShapeExtents {x=0}}}
 }
 {id=Delay_0, children [Random_1], properties {delayMax=30, delayMin=5}}
 {
  id=Enveloper_0
  children [WavePlayer_3]
  properties {
   bLoop=true
   bLoopIndefinitely=true
   loopEnd=60
   pitchCurve {editorCurveData {keys [{time=0}]}}
   volumeCurve {editorCurveData {keys [{arriveTangent=-0.0017, time=0, value=0.8}, <at 30 1>, <at 60 0.8>]}}
  }
 }
 {id=Looping_0, children [Delay_0]}
 {id=Mixer_1, children [Attenuation_1, Attenuation_0], properties {inputVolume [1,1]}}
 {id=Random_1, children [WavePlayer_0,WavePlayer_5,WavePlayer_6,WavePlayer_7], properties {weights [1,1,1,1]}}
 {id=WavePlayer_0, properties {soundWaveAssetPtr=/Game/Athena/Sounds/Events/RiftBeacon/AssetDrop_01/Rift_Energy_Dist_04}}
 {id=WavePlayer_3, properties {bLooping=true, soundWaveAssetPtr=&/Meteor_AmbientLoop_04}}
 {id=WavePlayer_5, properties {soundWaveAssetPtr=/Game/Athena/Sounds/Events/RiftBeacon/AssetDrop_01/Rift_Energy_Dist_01}}
 {id=WavePlayer_6, properties {soundWaveAssetPtr=/Game/Athena/Sounds/Events/RiftBeacon/AssetDrop_01/Rift_Energy_Dist_02}}
 {id=WavePlayer_7, properties {soundWaveAssetPtr=/Game/Athena/Sounds/Events/RiftBeacon/AssetDrop_01/Rift_Energy_Dist_03}}
]
properties {
 maxDistance=20400
 soundClassObject=/Figment_S10/Athena/Sounds/Events/SoundClasses/Fort_Events_Priority_ExtHigh
 volumeMultiplier=1
 attenuationOverrides {
  bEnableFocusInterpolation=true
  bEnableListenerFocus=true
  distanceAlgorithm=Logarithmic
  falloffDistance=20000
  focusAttackInterpSpeed=0.5
  focusAzimuth=5
  focusDistanceScale=0.5
  focusReleaseInterpSpeed=0.25
  lPFFrequencyAtMax=500
  lPFRadiusMax=2000
  lPFRadiusMin=1000
  nonFocusAzimuth=30
  nonFocusVolumeAttenuation=0.5
  stereoSpread=4096
  attenuationShapeExtents {x=0}
  customAttenuationCurve {editorCurveData {keys [<at 0.1 0>,<at 0.25 1>,<at 1 0>]}}
 }
}
