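# Ambient Sound Cue for the Figment Season 10 meteor event, looping a reused Creative rift close-loop sound under a high-priority events sound class.
# A Mixer feeds two Attenuation branches, but only one has a Wave Player; the wider-falloff branch is empty.
# The audible branch overrides attenuation with a large shape, long falloff, wide stereo spread and a low-pass filter that closes in with distance.
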SoundCue /Figment_S10/Athena/Sounds/Events/Meteor/Meteor_Ambient_03_Cue
root=Mixer_1
nodes [
 {
  id=Attenuation_0
  children [WavePlayer_8]
  properties {
   bOverrideAttenuation=true
   attenuationOverrides {
    bAttenuateWithLPF=true
    falloffDistance=15000
    lPFFrequencyAtMax=2000
    lPFRadiusMax=10000
    lPFRadiusMin=7000
    stereoSpread=4096
    attenuationShapeExtents {x=4707.0684}
   }
  }
 }
 {
  id=Attenuation_1
  children [null]
  properties {bOverrideAttenuation=true, attenuationOverrides {falloffDistance=20000, stereoSpread=4096, attenuationShapeExtents {x=0}}}
 }
 {id=Mixer_1, children [Attenuation_1, Attenuation_0], properties {inputVolume [1,1]}}
 {id=WavePlayer_8, properties {bLooping=true, soundWaveAssetPtr=/Game/Sounds/Creative/Rift/Creative_Rift_CloseLoop_New}}
]
properties {
 maxDistance=20000
 soundClassObject=/Figment_S10/Athena/Sounds/Events/SoundClasses/Fort_Events_Priority_High
 volumeMultiplier=1.5
 attenuationOverrides {
  bEnableFocusInterpolation=true
  bEnableListenerFocus=true
  distanceAlgorithm=Logarithmic
  falloffDistance=20000
  focusAttackInterpSpeed=0.5
  focusAzimuth=5
  focusDistanceScale=0.5
  focusReleaseInterpSpeed=0.25
  lPFFrequencyAtMax=500
  lPFRadiusMax=2000
  lPFRadiusMin=1000
  nonFocusAzimuth=30
  nonFocusVolumeAttenuation=0.5
  stereoSpread=4096
  attenuationShapeExtents {x=0}
  customAttenuationCurve {editorCurveData {keys [<at 0.1 0>,<at 0.25 1>,<at 1 0>]}}
 }
}
