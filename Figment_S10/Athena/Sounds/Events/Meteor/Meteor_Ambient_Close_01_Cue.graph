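# Sound cue for the close-range looping ambience of the meteor event, played in a high-priority events sound class.
# A mixer layers a boosted AmbientLoop_01 with AmbientLoop_02, whose volume slowly dips and recovers through an endlessly looping one-minute envelope.
# Overridden attenuation uses a custom curve that reaches silence a tenth of the way into a very long falloff, with wide stereo spread and interpolated listener focus.
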SoundCue /Figment_S10/Athena/Sounds/Events/Meteor/Meteor_Ambient_Close_01_Cue
root=Mixer_1
nodes [
 {
  id=Enveloper_0
  children [WavePlayer_4]
  properties {
   bLoop=true
   bLoopIndefinitely=true
   loopEnd=60
   pitchCurve {editorCurveData {keys [{time=0}]}}
   volumeCurve {editorCurveData {keys [<at 0 1>,<at 30 0.75>,<at 60 1>]}}
  }
 }
 {id=Mixer_1, children [WavePlayer_3,Enveloper_0], properties {inputVolume [2.25,1]}}
 {id=WavePlayer_3, properties {bLooping=true, soundWaveAssetPtr=&/Meteor_AmbientLoop_01}}
 {id=WavePlayer_4, properties {bLooping=true, soundWaveAssetPtr=&/Meteor_AmbientLoop_02}}
]
properties {
 bOverrideAttenuation=true
 maxDistance=101024
 soundClassObject=/Figment_S10/Athena/Sounds/Events/SoundClasses/Fort_Events_Priority_ExtHigh
 volumeMultiplier=1.5
 attenuationOverrides {
  bEnableFocusInterpolation=true
  bEnableListenerFocus=true
  distanceAlgorithm=Custom
  falloffDistance=100000
  focusAttackInterpSpeed=0.25
  focusAzimuth=5
  focusDistanceScale=0.3
  focusReleaseInterpSpeed=0.25
  lPFFrequencyAtMax=500
  lPFRadiusMax=2000
  lPFRadiusMin=1000
  nonFocusAzimuth=10
  nonFocusVolumeAttenuation=0.5
  stereoSpread=2048
  attenuationShapeExtents {x=1024}
  customAttenuationCurve {editorCurveData {keys [<at 0 1>,<at 0.1 0>,<at 1 0>]}}
 }
}
