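# Sound cue for the Meteor event's ambient glitch one-shots, routed to a high-priority events sound class.
# A Random node picks one of seven equally weighted glitch waves and feeds it through a Modulator for pitch and volume variation.
# Overridden attenuation gives long-range logarithmic falloff with a distance low-pass filter, wide stereo spread, and interpolated listener focus that quiets off-axis sources.
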SoundCue /Figment_S10/Athena/Sounds/Events/Meteor/Meteor_Ambient_Glitch_Cue
root=Modulator_0
nodes [
 {id=Modulator_0, children [Random_0]}
 {
  id=Random_0
  children [WavePlayer_0, WavePlayer_5, WavePlayer_6, WavePlayer_7, WavePlayer_8, WavePlayer_9, WavePlayer_10]
  properties {weights [1,1,1,1,1,1,1]}
 }
 {id=WavePlayer_0, properties {soundWaveAssetPtr=&/Meteor_Glitch_Oneshot_01}}
 {id=WavePlayer_10, properties {soundWaveAssetPtr=&/Meteor_Glitch_Oneshot_07}}
 {id=WavePlayer_5, properties {soundWaveAssetPtr=&/Meteor_Glitch_Oneshot_02}}
 {id=WavePlayer_6, properties {soundWaveAssetPtr=&/Meteor_Glitch_Oneshot_03}}
 {id=WavePlayer_7, properties {soundWaveAssetPtr=&/Meteor_Glitch_Oneshot_04}}
 {id=WavePlayer_8, properties {soundWaveAssetPtr=&/Meteor_Glitch_Oneshot_05}}
 {id=WavePlayer_9, properties {soundWaveAssetPtr=&/Meteor_Glitch_Oneshot_06}}
]
properties {
 bOverrideAttenuation=true
 maxDistance=20000
 soundClassObject=/Figment_S10/Athena/Sounds/Events/SoundClasses/Fort_Events_Priority_ExtHigh
 volumeMultiplier=2
 attenuationOverrides {
  bAttenuateWithLPF=true
  bEnableFocusInterpolation=true
  bEnableListenerFocus=true
  distanceAlgorithm=Logarithmic
  falloffDistance=20000
  focusAttackInterpSpeed=0.5
  focusAzimuth=5
  focusDistanceScale=0.5
  focusReleaseInterpSpeed=0.25
  lPFFrequencyAtMax=500
  lPFRadiusMax=20000
  lPFRadiusMin=10000
  nonFocusAzimuth=30
  nonFocusVolumeAttenuation=0.5
  stereoSpread=4096
  attenuationShapeExtents {x=0}
  customAttenuationCurve {editorCurveData {keys [<at 0.1 0>,<at 0.25 1>,<at 1 0>]}}
 }
}
