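# Sound cue for a small meteor ambient in the Figment S10 events, played as an endless loop of Meteor_AmbientLoop_05.
# A single looping wave player feeds the output with boosted volume, routed to the high-priority external events sound class.
# Overridden attenuation gives short-range logarithmic falloff from the source, with distance low-pass filtering and occlusion muffling.
# Interpolated listener focus lowers the volume when the meteor is off-axis.
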SoundCue /Figment_S10/Athena/Sounds/Events/Meteor/Meteor_Ambient_Small_01_Cue
root=WavePlayer_3
nodes [{id=WavePlayer_3, properties {bLooping=true, soundWaveAssetPtr=&/Meteor_AmbientLoop_05}}]
properties {
 bOverrideAttenuation=true
 maxDistance=1024
 soundClassObject=/Figment_S10/Athena/Sounds/Events/SoundClasses/Fort_Events_Priority_ExtHigh
 volumeMultiplier=1.5
 attenuationOverrides {
  bAttenuateWithLPF=true
  bEnableFocusInterpolation=true
  bEnableListenerFocus=true
  bEnableOcclusion=true
  distanceAlgorithm=Logarithmic
  falloffDistance=1024
  focusAzimuth=5
  lPFFrequencyAtMax=500
  lPFRadiusMax=2000
  lPFRadiusMin=1000
  nonFocusAzimuth=30
  nonFocusVolumeAttenuation=0.25
  occlusionLowPassFilterFrequency=1000
  occlusionTraceChannel=ECC_GameTraceChannel11
  stereoSpread=512
  attenuationShapeExtents {x=0}
  customAttenuationCurve {editorCurveData {keys [<at 0.1 0>,<at 0.25 1>,<at 1 0>]}}
 }
}
